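# Physics asset for the male medium TikiTorch player body cosmetic, covering the lower body and a kimono-style skirt.
# Five three-bone skirt chains (back center, back left/right, side left/right) each constrain to their own MSTR root body.
# Constraints act like limited hinges, allowing modest swing on one axis with the other swing and twist nearly locked.
# Unconstrained capsules cover the pelvis, legs, feet and skirt master, with some left/right size mismatches.
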PhysicsAsset /BRCosmetics/Characters/Player/Male/Medium/Bodies/M_MED_TikiTorch/Meshes/M_MED_TikiTorch_Physics
bodies {
 calf_l [{shapeType=Capsule, length=14, radius=7, shapeName=calf_r_capsule, center=<-7 -1.5 0>, rotation=<rot 90 0 -10>}]
 calf_r [{shapeType=Capsule, length=14, radius=10, shapeName=calf_r_capsule, center=<7 1.5 0>, rotation=<rot 90 0 -10>}]
 calf_twist_01_l [{shapeType=Capsule, length=24, radius=6.5, shapeName=calf_r_capsule, center=<7 -1 0>, rotation=<rot 90 0 5>}]
 calf_twist_01_r [{shapeType=Capsule, length=24, radius=6.5, shapeName=calf_r_capsule, center=<-7 1 0>, rotation=<rot 90 0 5>}]
 dyn_kimono_back_01 [{shapeType=Capsule, length=10, radius=3, shapeName=dyn_kimono_front_03_r_capsule, center=<7.1013 0 0>}]
 dyn_kimono_back_01_MSTR [{shapeType=Capsule, length=4, radius=1, shapeName=dyn_kimono_back_01_MSTR_capsule}]
 dyn_kimono_back_01_l [{shapeType=Capsule, length=10, radius=3, shapeName=dyn_kimono_front_03_r_capsule, center=<7.1013 0 0>}]
 dyn_kimono_back_01_l_MSTR [{shapeType=Capsule, length=4, radius=1, shapeName=dyn_kimono_back_01_l_MSTR_capsule}]
 dyn_kimono_back_01_r [{shapeType=Capsule, length=10, radius=3, shapeName=dyn_kimono_front_03_r_capsule, center=<7.1013 0 0>}]
 dyn_kimono_back_01_r_MSTR [{shapeType=Capsule, length=4, radius=1, shapeName=dyn_kimono_back_01_r_MSTR_capsule}]
 dyn_kimono_back_02 [{shapeType=Capsule, length=10, radius=6, shapeName=dyn_kimono_front_03_r_capsule, center=<6.6518 0 0>}]
 dyn_kimono_back_02_l [{shapeType=Capsule, length=10, radius=6, shapeName=dyn_kimono_front_03_r_capsule, center=<6.6518 0 0>}]
 dyn_kimono_back_02_r [{shapeType=Capsule, length=10, radius=6, shapeName=dyn_kimono_front_03_r_capsule, center=<6.6518 0 0>}]
 dyn_kimono_back_03 [{shapeType=Capsule, length=10, radius=6, shapeName=dyn_kimono_front_03_r_capsule, center=<6.6518 0 0>}]
 dyn_kimono_back_03_l [{shapeType=Capsule, length=10, radius=6, shapeName=dyn_kimono_front_03_r_capsule, center=<6.6518 0 0>}]
 dyn_kimono_back_03_r [{shapeType=Capsule, length=20, radius=8, shapeName=dyn_kimono_front_03_r_capsule, center=<1.7345 0 0>}]
 dyn_kimono_side_01_l [{shapeType=Capsule, length=10, radius=3, shapeName=dyn_kimono_front_03_r_capsule, center=<7.1013 0 0>}]
 dyn_kimono_side_01_l_MSTR [{shapeType=Capsule, length=4, radius=1, shapeName=dyn_kimono_side_01_l_MSTR_capsule}]
 dyn_kimono_side_01_r [{shapeType=Capsule, length=10, radius=3, shapeName=dyn_kimono_front_03_r_capsule, center=<7.1013 0 0>}]
 dyn_kimono_side_01_r_MSTR [{shapeType=Capsule, length=4, radius=1, shapeName=dyn_kimono_side_01_r_MSTR_capsule}]
 dyn_kimono_side_02_l [{shapeType=Capsule, length=10, radius=6, shapeName=dyn_kimono_front_03_r_capsule, center=<6.6518 0 0>}]
 dyn_kimono_side_02_r [{shapeType=Capsule, length=10, radius=6, shapeName=dyn_kimono_front_03_r_capsule, center=<6.6518 0 0>}]
 dyn_kimono_side_03_l [{shapeType=Capsule, length=10, radius=6, shapeName=dyn_kimono_front_03_r_capsule, center=<6.6518 0 0>}]
 dyn_kimono_side_03_r [{shapeType=Capsule, length=10, radius=6, shapeName=dyn_kimono_front_03_r_capsule, center=<6.6518 0 0>}]
 foot_l [{
   shapeType=Capsule
   length=18.8263
   radius=7.899
   shapeName=foot_r_capsule
   center=<-3.8984 6.0897 1.0346>
   rotation=<rot 0.0079 -0.4187 -93.2>
  }]
 foot_r [{
   shapeType=Capsule
   length=18.8263
   radius=7.899
   shapeName=foot_r_capsule
   center=<3.3729 -5.8893 0.3689>
   rotation=<rot 0.0079 -0.4187 -93.2>
  }]
 kimono_skirt_MSTR [{shapeType=Capsule, length=15, radius=1, shapeName=kimono_skirt_MSTR_capsule}]
 pelvis [{shapeType=Capsule, length=7, radius=10.4, shapeName=spine_01_capsule, center=<0 -0.7 0>}]
 thigh_l [{shapeType=Capsule, length=20, radius=8, shapeName=thigh_l_capsule, center=<-6 0 1>, rotation=<rot 95 0 0>}]
 thigh_r [{shapeType=Capsule, length=20, radius=8, shapeName=thigh_l_capsule, center=<6 0 -1>, rotation=<rot 95 0 0>}]
 thigh_twist_01_l [{shapeType=Capsule, length=14, radius=7, shapeName=thigh_twist_01_l_capsule, center=<-4 0 0>, rotation=<rot 90 0 0>}]
 thigh_twist_01_r [{shapeType=Capsule, length=35, radius=7, shapeName=thigh_twist_01_r_capsule, center=<-4 0 0>, rotation=<rot 90 0 0>}]
}
constraints [
 {bone1Name=dyn_kimono_back_03, bone2Name=dyn_kimono_back_02, swing1LimitDegrees=5, swing2LimitDegrees=1, twistLimitDegrees=1}
 {bone1Name=dyn_kimono_back_02, bone2Name=dyn_kimono_back_01, swing1LimitDegrees=5, swing2LimitDegrees=1, twistLimitDegrees=1}
 {bone1Name=dyn_kimono_back_01, bone2Name=dyn_kimono_back_01_MSTR, swing1LimitDegrees=30, swing2LimitDegrees=1, twistLimitDegrees=1}
 {bone1Name=dyn_kimono_back_03_l, bone2Name=dyn_kimono_back_02_l, swing1LimitDegrees=15, swing2LimitDegrees=1, twistLimitDegrees=1}
 {bone1Name=dyn_kimono_back_02_l, bone2Name=dyn_kimono_back_01_l, swing1LimitDegrees=15, swing2LimitDegrees=1, twistLimitDegrees=1}
 {bone1Name=dyn_kimono_back_01_l, bone2Name=dyn_kimono_back_01_l_MSTR, swing1LimitDegrees=15, swing2LimitDegrees=1, twistLimitDegrees=1}
 {bone1Name=dyn_kimono_back_03_r, bone2Name=dyn_kimono_back_02_r, swing1LimitDegrees=15, swing2LimitDegrees=1, twistLimitDegrees=1}
 {bone1Name=dyn_kimono_back_02_r, bone2Name=dyn_kimono_back_01_r, swing1LimitDegrees=15, swing2LimitDegrees=1, twistLimitDegrees=1}
 {bone1Name=dyn_kimono_back_01_r, bone2Name=dyn_kimono_back_01_r_MSTR, swing1LimitDegrees=15, swing2LimitDegrees=1, twistLimitDegrees=1}
 {bone1Name=dyn_kimono_side_03_l, bone2Name=dyn_kimono_side_02_l, swing1LimitDegrees=30, swing2LimitDegrees=1, twistLimitDegrees=1}
 {bone1Name=dyn_kimono_side_02_l, bone2Name=dyn_kimono_side_01_l, swing1LimitDegrees=15, swing2LimitDegrees=1, twistLimitDegrees=1}
 {bone1Name=dyn_kimono_side_01_l, bone2Name=dyn_kimono_side_01_l_MSTR, swing1LimitDegrees=15, swing2LimitDegrees=1, twistLimitDegrees=1}
 {bone1Name=dyn_kimono_side_03_r, bone2Name=dyn_kimono_side_02_r, swing1LimitDegrees=15, swing2LimitDegrees=1, twistLimitDegrees=1}
 {bone1Name=dyn_kimono_side_02_r, bone2Name=dyn_kimono_side_01_r, swing1LimitDegrees=15, swing2LimitDegrees=1, twistLimitDegrees=1}
 {bone1Name=dyn_kimono_side_01_r, bone2Name=dyn_kimono_side_01_r_MSTR, swing1LimitDegrees=15, swing2LimitDegrees=1, twistLimitDegrees=1}
]
