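# Post-process AnimBP for the TikiTorch cosmetic body, adding secondary motion to kimono and sleeves over the pose copied from the parent mesh.
# AnimDynamics drives front flaps and, outside front-end idle, sleeves with spherical limits and arm-angle-driven offsets.
# Skirt bones then use either RigidBody with state-based gravity or thigh/pelvis constraints, selected by bRigidBodyEnabled, with movement-state corrections.
# Several sleeve branches do not reach the output.
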AnimBlueprint /BRCosmetics/Characters/Player/Male/Medium/Bodies/M_MED_TikiTorch/Meshes/M_MED_TikiTorch_AnimBP
target_skeleton=&/M_MED_TikiTorch_Skeleton
graphs [
 AnimGraph {nodes [
   n1 = Anim Dynamics - Bone: dyn_sleev_l!(ComponentPose, bAlphaBoolEnabled:bool=True, ExternalForce="0,0,0") -> (Pose) {
    alphaInputType=Bool
    angularDampingOverride=1
    bOverrideAngularDamping=true
    bUseGravityOverride=true
    simulationSpace=RootRelative
    alphaBoolBlend {blendInTime=0.2, blendOption=Linear, blendOutTime=0.4}
    boundBone {boneName=dyn_sleev_l}
    componentAppliedLinearAccClamp=<3000 3000 3000>
    componentLinearAccScale=<0.1 0.1 0.1>
    componentLinearVelScale=<0.2 0.2 0.2>
    gravityOverride {z=-980}
    physicsBodyDefinitions [{
      boundBone {boneName=dyn_sleev_l}
      boxExtents=<5 15 15>
      constraintSetup {angularLimitsMax=<0 10 10>, angularLimitsMin=<0 -10 -10>, angularTarget=<0 0 0>}
      localJointOffset=<2 0 0>
     }]
   }
   Anim Dynamics - Bone: dyn_sleev_r!(ComponentPose=@n1, bAlphaBoolEnabled:bool=True, ExternalForce="0,0,0") -> (Pose) {
    alphaInputType=Bool
    angularDampingOverride=1
    bOverrideAngularDamping=true
    bUseGravityOverride=true
    simulationSpace=RootRelative
    alphaBoolBlend {blendInTime=0.2, blendOption=Linear, blendOutTime=0.4}
    boundBone {boneName=dyn_sleev_r}
    componentAppliedLinearAccClamp=<3000 3000 3000>
    componentLinearAccScale=<0.1 0.1 0.1>
    componentLinearVelScale=<0.2 0.2 0.2>
    gravityOverride {z=-980}
    physicsBodyDefinitions [{
      boundBone {boneName=dyn_sleev_r}
      boxExtents=<5 15 15>
      constraintSetup {angularLimitsMax=<0 10 5>, angularLimitsMin=<0 -10 -10>, angularTarget=<0 0 0>}
      localJointOffset=<-2 0 0>
     }]
   }
   n2 = Get Aim_offset_l() -> (Aim_offset_l:float)
   n3 = "Transform (Modify) Bone - Bone: dyn_sleevs_test_01_MSTR"(
    ComponentPose=Local To Component(LocalPose=Use cached pose 'PreSimSleeves'())
    Alpha=@n2
    Translation="-4,0,0"
    Rotation="0,0,0"
    Scale="1,1,1"
   ) {
    rotationMode=BMM_Additive
    rotationSpace=BCS_BoneSpace
    translationMode=BMM_Additive
    translationSpace=BCS_BoneSpace
    alphaBoolBlend {blendInTime=0.1, blendOption=Linear, blendOutTime=0.1}
    boneToModify {boneName=dyn_sleevs_test_01_MSTR}
    translation {x=-4}
   }
   n4 = "Transform (Modify) Bone - Bone: dyn_sleevs_test_04_MSTR"(ComponentPose=@n3, Alpha=@n2, Translation="-12,0,0", Rotation="0,0,0", Scale="1,1,1") {
    rotationMode=BMM_Additive
    rotationSpace=BCS_BoneSpace
    translationMode=BMM_Additive
    translationSpace=BCS_BoneSpace
    alphaBoolBlend {blendInTime=0.1, blendOption=Linear, blendOutTime=0.1}
    boneToModify {boneName=dyn_sleevs_test_04_MSTR}
    translation {x=-12}
   }
   Blend Poses by bool(BlendPose_0=Component To Local(ComponentPose=@n4), BlendPose_1=Component To Local(), bActiveValue=Get bIsPlayingEmote()) {blendType=Linear}
   n5 = "Transform (Modify) Bone - Bone: dyn_sleev_rot_r"(
    ComponentPose=Local To Component(LocalPose=Use cached pose 'First_stage'())
    Alpha=Get Aim_offset_r() -> (Aim_offset_r:float)
    Translation="0,0,0"
    Rotation="0,52,0"
    Scale="1,1,1"
   ) {
    rotationMode=BMM_Additive
    rotationSpace=BCS_BoneSpace
    alphaBoolBlend {blendOption=Linear}
    boneToModify {boneName=dyn_sleev_rot_r}
    rotation {yaw=52}
   }
   n6 = "Transform (Modify) Bone - Bone: dyn_sleev_rot_l"(ComponentPose=@n5, Alpha=Get Aim_offset_l(), Translation="0,0,0", Rotation="0,50,0", Scale="1,1,1") {
    rotationMode=BMM_Additive
    rotationSpace=BCS_BoneSpace
    alphaBoolBlend {blendOption=Linear}
    boneToModify {boneName=dyn_sleev_rot_l}
    rotation {yaw=50}
   }
   Local To Component(LocalPose=Blend Poses by bool(BlendPose_0=Component To Local(), BlendPose_1=Component To Local(ComponentPose=@n6)) {blendType=Linear})
   n7 = Get Aim_offset_r()
   "Transform (Modify) Bone - Bone: dyn_sleevs_r_03_MSTR"(Alpha=@n7, Translation="-15,0,0", Rotation="0,0,0", Scale="1,1,1") {
    rotationMode=BMM_Additive
    rotationSpace=BCS_BoneSpace
    translationMode=BMM_Additive
    translationSpace=BCS_BoneSpace
    alphaBoolBlend {blendInTime=0.1, blendOption=Linear, blendOutTime=0.1}
    boneToModify {boneName=dyn_sleevs_r_03_MSTR}
    translation {x=-15}
   }
   n8 = Get Aim_offset_l()
   "Transform (Modify) Bone - Bone: dyn_sleevs_l_03_MSTR"(Alpha=@n8, Translation="-15,0,0", Rotation="0,0,0", Scale="1,1,1") {
    rotationMode=BMM_Additive
    rotationSpace=BCS_BoneSpace
    translationMode=BMM_Additive
    translationSpace=BCS_BoneSpace
    alphaBoolBlend {blendInTime=0.1, blendOption=Linear, blendOutTime=0.1}
    boneToModify {boneName=dyn_sleevs_l_03_MSTR}
    translation {x=-15}
   }
   "Transform (Modify) Bone - Bone: dyn_sleevs_test_03_MSTR"(Alpha=@n2, Translation="-8,0,0", Rotation="0,0,0", Scale="1,1,1") {
    rotationMode=BMM_Additive
    rotationSpace=BCS_BoneSpace
    translationMode=BMM_Additive
    translationSpace=BCS_BoneSpace
    alphaBoolBlend {blendInTime=0.1, blendOption=Linear, blendOutTime=0.1}
    boneToModify {boneName=dyn_sleevs_test_03_MSTR}
    translation {x=-8}
   }
   n9 = Get bIsInFrontEnd()
   n10 = Get bIsSliding()
   n11 = Get AltSkinMovement() -> (AltSkinMovement:bool)
   n12 = Get bIsCrouching()
   n13 = "Constraint - dyn_kimono_back_01_l_MSTR (RR)"!(
    ComponentPose=Local To Component(LocalPose=Use cached pose 'First_stage'())
    bAlphaBoolEnabled:bool=@n12
    ConstraintWeights_0=0.3
    ConstraintWeights_1:float=0.7
   ) -> (Pose) {
    alphaInputType=Bool
    alphaBoolBlend {blendInTime=0.1, blendOption=Linear, blendOutTime=0.2}
    boneToModify {boneName=dyn_kimono_back_01_l_MSTR}
    constraintSetup [{transformType=Rotation, targetBone {boneName=thigh_l}}, {transformType=Rotation, targetBone {boneName=pelvis}}]
    constraintWeights [1,1]
   }
   n14 = "Constraint - dyn_kimono_back_01_MSTR (RRR)"!(ComponentPose=@n13, bAlphaBoolEnabled:bool=@n12, ConstraintWeights_0=0.8, ConstraintWeights_1:float=0.1, ConstraintWeights_2:float=0.1) -> (Pose) {
    alphaInputType=Bool
    alphaBoolBlend {blendInTime=0.1, blendOption=Linear, blendOutTime=0.2}
    boneToModify {boneName=dyn_kimono_back_01_MSTR}
    constraintSetup [
     {transformType=Rotation, targetBone {boneName=pelvis}}
     {transformType=Rotation, targetBone {boneName=thigh_l}}
     {transformType=Rotation, targetBone {boneName=thigh_r}}
    ]
    constraintWeights [1,1,1]
   }
   n15 = "Constraint - dyn_kimono_back_01_r_MSTR (RR)"!(ComponentPose=@n14, bAlphaBoolEnabled:bool=@n12, ConstraintWeights_0=0.3, ConstraintWeights_1:float=0.7) -> (Pose) {
    alphaInputType=Bool
    alphaBoolBlend {blendInTime=0.1, blendOption=Linear, blendOutTime=0.2}
    boneToModify {boneName=dyn_kimono_back_01_r_MSTR}
    constraintSetup [{transformType=Rotation, targetBone {boneName=thigh_r}}, {transformType=Rotation, targetBone {boneName=pelvis}}]
    constraintWeights [1,1]
   }
   n16 = "Transform (Modify) Bone - Bone: dyn_kimono_back_01_l_MSTR"!(ComponentPose=@n15, bAlphaBoolEnabled:bool=@n12, Translation="0,0,0", Rotation="0,-15,0", Scale="1,1,1") -> (Pose) {
    alphaInputType=Bool
    rotationMode=BMM_Additive
    rotationSpace=BCS_BoneSpace
    alphaBoolBlend {blendInTime=0.1, blendOption=Linear, blendOutTime=0.2}
    boneToModify {boneName=dyn_kimono_back_01_l_MSTR}
    rotation {yaw=-15}
   }
   n17 = "Transform (Modify) Bone - Bone: dyn_kimono_back_01_MSTR"!(ComponentPose=@n16, bAlphaBoolEnabled:bool=@n12, Translation="0,0,0", Rotation="0,-15,0", Scale="1,1,1") -> (Pose) {
    alphaInputType=Bool
    rotationMode=BMM_Additive
    rotationSpace=BCS_BoneSpace
    translationMode=BMM_Additive
    translationSpace=BCS_BoneSpace
    alphaBoolBlend {blendInTime=0.1, blendOption=Linear, blendOutTime=0.2}
    boneToModify {boneName=dyn_kimono_back_01_MSTR}
    rotation {yaw=-15}
   }
   n18 = "Transform (Modify) Bone - Bone: dyn_kimono_back_01_r_MSTR"!(ComponentPose=@n17, bAlphaBoolEnabled:bool=@n12, Translation="0,0,0", Rotation="0,-35,0", Scale="1,1,1") -> (Pose) {
    alphaInputType=Bool
    rotationMode=BMM_Additive
    rotationSpace=BCS_BoneSpace
    alphaBoolBlend {blendInTime=0.1, blendOption=Linear, blendOutTime=0.2}
    boneToModify {boneName=dyn_kimono_back_01_r_MSTR}
    rotation {yaw=-35}
   }
   n19 = "Constraint - dyn_kimono_back_01_l_MSTR (RR)"(ComponentPose=@n18, bAlphaBoolEnabled=@n11, ConstraintWeights_0=0, ConstraintWeights_1=1) {
    alphaInputType=Bool
    alphaBoolBlend {blendInTime=0.1, blendOption=Linear, blendOutTime=0.2}
    boneToModify {boneName=dyn_kimono_back_01_l_MSTR}
    constraintSetup [{transformType=Rotation, targetBone {boneName=thigh_l}}, {transformType=Rotation, targetBone {boneName=pelvis}}]
    constraintWeights [1,1]
   }
   n20 = "Constraint - dyn_kimono_back_01_MSTR (RRR)"(ComponentPose=@n19, bAlphaBoolEnabled=@n11, ConstraintWeights_0=1, ConstraintWeights_1=0, ConstraintWeights_2=0) {
    alphaInputType=Bool
    alphaBoolBlend {blendInTime=0.1, blendOption=Linear, blendOutTime=0.2}
    boneToModify {boneName=dyn_kimono_back_01_MSTR}
    constraintSetup [
     {transformType=Rotation, targetBone {boneName=pelvis}}
     {transformType=Rotation, targetBone {boneName=thigh_l}}
     {transformType=Rotation, targetBone {boneName=thigh_r}}
    ]
    constraintWeights [1,1,1]
   }
   n21 = "Constraint - dyn_kimono_back_01_r_MSTR (RR)"(ComponentPose=@n20, bAlphaBoolEnabled=@n11, ConstraintWeights_0=0, ConstraintWeights_1=1) {
    alphaInputType=Bool
    alphaBoolBlend {blendInTime=0.1, blendOption=Linear, blendOutTime=0.2}
    boneToModify {boneName=dyn_kimono_back_01_r_MSTR}
    constraintSetup [{transformType=Rotation, targetBone {boneName=thigh_r}}, {transformType=Rotation, targetBone {boneName=pelvis}}]
    constraintWeights [1,1]
   }
   n22 = "Transform (Modify) Bone - Bone: dyn_kimono_back_01_l_MSTR"(ComponentPose=@n21, bAlphaBoolEnabled=@n11, Translation="0,0,0", Rotation="0,-15,0", Scale="1,1,1") {
    alphaInputType=Bool
    rotationMode=BMM_Additive
    rotationSpace=BCS_BoneSpace
    translationMode=BMM_Additive
    translationSpace=BCS_BoneSpace
    alphaBoolBlend {blendInTime=0.1, blendOption=Linear, blendOutTime=0.2}
    boneToModify {boneName=dyn_kimono_back_01_l_MSTR}
    rotation {yaw=-15}
   }
   n23 = "Transform (Modify) Bone - Bone: dyn_kimono_back_01_MSTR"(ComponentPose=@n22, bAlphaBoolEnabled=@n11, Translation="0,0,0", Rotation="0,-15,0", Scale="1,1,1") {
    alphaInputType=Bool
    rotationMode=BMM_Additive
    rotationSpace=BCS_BoneSpace
    translationMode=BMM_Additive
    translationSpace=BCS_BoneSpace
    alphaBoolBlend {blendInTime=0.1, blendOption=Linear, blendOutTime=0.2}
    boneToModify {boneName=dyn_kimono_back_01_MSTR}
    rotation {yaw=-15}
   }
   n24 = "Transform (Modify) Bone - Bone: dyn_kimono_back_01_r_MSTR"(ComponentPose=@n23, bAlphaBoolEnabled=@n11, Translation="0,0,0", Rotation="0,-25,0", Scale="1,1,1") {
    alphaInputType=Bool
    rotationMode=BMM_Additive
    rotationSpace=BCS_BoneSpace
    translationMode=BMM_Additive
    translationSpace=BCS_BoneSpace
    alphaBoolBlend {blendInTime=0.1, blendOption=Linear, blendOutTime=0.2}
    boneToModify {boneName=dyn_kimono_back_01_r_MSTR}
    rotation {yaw=-25}
   }
   n25 = "Constraint - dyn_kimono_side_01_r_MSTR (RR)"!(ComponentPose=@n24, bAlphaBoolEnabled:bool=@n10, ConstraintWeights_0=0.7, ConstraintWeights_1:float=0.3) -> (Pose) {
    alphaInputType=Bool
    alphaBoolBlend {blendInTime=0.1, blendOption=Linear, blendOutTime=0.2}
    boneToModify {boneName=dyn_kimono_side_01_r_MSTR}
    constraintSetup [{transformType=Rotation, targetBone {boneName=thigh_r}}, {transformType=Rotation, targetBone {boneName=pelvis}}]
    constraintWeights [1,1]
   }
   n26 = "Transform (Modify) Bone - Bone: dyn_kimono_side_01_r_MSTR"!(ComponentPose=@n25, bAlphaBoolEnabled:bool=@n10, Translation="0,-5,0", Rotation="0,-15,0", Scale="1,1,1") -> (Pose) {
    alphaInputType=Bool
    rotationMode=BMM_Additive
    rotationSpace=BCS_BoneSpace
    translationMode=BMM_Additive
    translationSpace=BCS_BoneSpace
    alphaBoolBlend {blendInTime=0.1, blendOption=Linear, blendOutTime=0.2}
    boneToModify {boneName=dyn_kimono_side_01_r_MSTR}
    rotation {yaw=-15}
    translation {y=-5}
   }
   n27 = Get bIsSkydiveDiveMode()
   n28 = "Constraint - dyn_kimono_back_01_l_MSTR (RR)"(ComponentPose=@n26, bAlphaBoolEnabled=@n27, ConstraintWeights_0=0.3, ConstraintWeights_1=0.7) {
    alphaInputType=Bool
    alphaBoolBlend {blendInTime=0.1, blendOption=Linear, blendOutTime=0.2}
    boneToModify {boneName=dyn_kimono_back_01_l_MSTR}
    constraintSetup [{transformType=Rotation, targetBone {boneName=thigh_l}}, {transformType=Rotation, targetBone {boneName=pelvis}}]
    constraintWeights [1,1]
   }
   n29 = "Constraint - dyn_kimono_back_01_MSTR (RRR)"(ComponentPose=@n28, bAlphaBoolEnabled=@n27, ConstraintWeights_0=0.4, ConstraintWeights_1=0.3, ConstraintWeights_2=0.3) {
    alphaInputType=Bool
    alphaBoolBlend {blendInTime=0.1, blendOption=Linear, blendOutTime=0.2}
    boneToModify {boneName=dyn_kimono_back_01_MSTR}
    constraintSetup [
     {transformType=Rotation, targetBone {boneName=pelvis}}
     {transformType=Rotation, targetBone {boneName=thigh_l}}
     {transformType=Rotation, targetBone {boneName=thigh_r}}
    ]
    constraintWeights [1,1,1]
   }
   n30 = "Constraint - dyn_kimono_back_01_r_MSTR (RR)"(ComponentPose=@n29, bAlphaBoolEnabled=@n27, ConstraintWeights_0=0.3, ConstraintWeights_1=0.7) {
    alphaInputType=Bool
    alphaBoolBlend {blendInTime=0.1, blendOption=Linear, blendOutTime=0.2}
    boneToModify {boneName=dyn_kimono_back_01_r_MSTR}
    constraintSetup [{transformType=Rotation, targetBone {boneName=thigh_r}}, {transformType=Rotation, targetBone {boneName=pelvis}}]
    constraintWeights [1,1]
   }
   n31 = "Constraint - dyn_kimono_side_01_r_MSTR (RR)"(ComponentPose=@n30, bAlphaBoolEnabled=Get bIsParachuteOpened(), ConstraintWeights_0=1, ConstraintWeights_1=0) {
    alphaInputType=Bool
    alphaBoolBlend {blendInTime=0.1, blendOption=Linear, blendOutTime=0.2}
    boneToModify {boneName=dyn_kimono_side_01_r_MSTR}
    constraintSetup [{transformType=Rotation, targetBone {boneName=thigh_r}}, {transformType=Rotation, targetBone {boneName=pelvis}}]
    constraintWeights [1,1]
   }
   n32 = "Transform (Modify) Bone - Bone: dyn_kimono_back_01_l_MSTR"(ComponentPose=@n31, bAlphaBoolEnabled=@n9, Translation="0,0,0", Rotation="0,5,0", Scale="1,1,1") {
    alphaInputType=Bool
    rotationMode=BMM_Additive
    rotationSpace=BCS_BoneSpace
    translationMode=BMM_Additive
    translationSpace=BCS_BoneSpace
    alphaBoolBlend {blendInTime=0.1, blendOption=Linear, blendOutTime=0.2}
    boneToModify {boneName=dyn_kimono_back_01_l_MSTR}
    rotation {yaw=5}
   }
   n33 = "Transform (Modify) Bone - Bone: dyn_kimono_back_01_MSTR"(ComponentPose=@n32, bAlphaBoolEnabled=@n9, Translation="0,0,0", Rotation="0,5,0", Scale="1,1,1") {
    alphaInputType=Bool
    rotationMode=BMM_Additive
    rotationSpace=BCS_BoneSpace
    translationMode=BMM_Additive
    translationSpace=BCS_BoneSpace
    alphaBoolBlend {blendInTime=0.1, blendOption=Linear, blendOutTime=0.2}
    boneToModify {boneName=dyn_kimono_back_01_MSTR}
    rotation {yaw=5}
   }
   n34 = "Transform (Modify) Bone - Bone: dyn_kimono_back_01_r_MSTR"(ComponentPose=@n33, bAlphaBoolEnabled=@n9, Translation="0,0,0", Rotation="0,5,0", Scale="1,1,1") {
    alphaInputType=Bool
    rotationMode=BMM_Additive
    rotationSpace=BCS_BoneSpace
    translationMode=BMM_Additive
    translationSpace=BCS_BoneSpace
    alphaBoolBlend {blendInTime=0.1, blendOption=Linear, blendOutTime=0.2}
    boneToModify {boneName=dyn_kimono_back_01_r_MSTR}
    rotation {yaw=5}
   }
   n35 = "Constraint - dyn_kimono_side_01_r_MSTR (RR)"(ComponentPose=@n34, bAlphaBoolEnabled=@n9, ConstraintWeights_0=0.7, ConstraintWeights_1=0.3) {
    alphaInputType=Bool
    alphaBoolBlend {blendInTime=0.1, blendOption=Linear, blendOutTime=0.2}
    boneToModify {boneName=dyn_kimono_side_01_r_MSTR}
    constraintSetup [{transformType=Rotation, targetBone {boneName=thigh_r}}, {transformType=Rotation, targetBone {boneName=pelvis}}]
    constraintWeights [1,1]
   }
   n36 = "Constraint - dyn_kimono_side_01_l_MSTR (RR)"!(ComponentPose=@n35, bAlphaBoolEnabled:bool=@n9, ConstraintWeights_0=0.7, ConstraintWeights_1:float=0.3) -> (Pose) {
    alphaInputType=Bool
    alphaBoolBlend {blendInTime=0.1, blendOption=Linear, blendOutTime=0.2}
    boneToModify {boneName=dyn_kimono_side_01_l_MSTR}
    constraintSetup [{transformType=Rotation, targetBone {boneName=thigh_l}}, {transformType=Rotation, targetBone {boneName=pelvis}}]
    constraintWeights [1,1]
   }
   n37 = Component To Local(ComponentPose=@n36)
   n38 = Get bIsCrouching()
   n39 = "Constraint - dyn_kimono_back_01_l_MSTR (RR)"(
    ComponentPose=Local To Component(LocalPose=Use cached pose 'First_stage'())
    bAlphaBoolEnabled=@n38
    ConstraintWeights_0=0.3
    ConstraintWeights_1=0.7
   ) {
    alphaInputType=Bool
    alphaBoolBlend {blendInTime=0.1, blendOption=Linear, blendOutTime=0.2}
    boneToModify {boneName=dyn_kimono_back_01_l_MSTR}
    constraintSetup [{transformType=Rotation, targetBone {boneName=thigh_l}}, {transformType=Rotation, targetBone {boneName=pelvis}}]
    constraintWeights [1,1]
   }
   n40 = "Constraint - dyn_kimono_back_01_MSTR (RR)"!(ComponentPose=@n39, bAlphaBoolEnabled:bool=@n38, ConstraintWeights_0=0.5, ConstraintWeights_1:float=0.5) -> (Pose) {
    alphaInputType=Bool
    alphaBoolBlend {blendInTime=0.1, blendOption=Linear, blendOutTime=0.2}
    boneToModify {boneName=dyn_kimono_back_01_MSTR}
    constraintSetup [{transformType=Rotation, targetBone {boneName=thigh_r}}, {transformType=Rotation, targetBone {boneName=thigh_l}}]
    constraintWeights [1,1]
   }
   n41 = "Constraint - dyn_kimono_back_01_r_MSTR (RR)"(ComponentPose=@n40, bAlphaBoolEnabled=@n38, ConstraintWeights_0=0.3, ConstraintWeights_1=0.7) {
    alphaInputType=Bool
    alphaBoolBlend {blendInTime=0.1, blendOption=Linear, blendOutTime=0.2}
    boneToModify {boneName=dyn_kimono_back_01_r_MSTR}
    constraintSetup [{transformType=Rotation, targetBone {boneName=thigh_r}}, {transformType=Rotation, targetBone {boneName=pelvis}}]
    constraintWeights [1,1]
   }
   n42 = "Transform (Modify) Bone - Bone: dyn_kimono_back_01_l_MSTR"(ComponentPose=@n41, bAlphaBoolEnabled=@n38, Translation="0,0,0", Rotation="0,30,0", Scale="1,1,1") {
    alphaInputType=Bool
    rotationMode=BMM_Additive
    rotationSpace=BCS_BoneSpace
    alphaBoolBlend {blendInTime=0.1, blendOption=Linear, blendOutTime=0.2}
    boneToModify {boneName=dyn_kimono_back_01_l_MSTR}
    rotation {yaw=30}
   }
   n43 = "Transform (Modify) Bone - Bone: dyn_kimono_back_01_r_MSTR"(ComponentPose=@n42, bAlphaBoolEnabled=@n38, Translation="0,0,0", Rotation="0,15,0", Scale="1,1,1") {
    alphaInputType=Bool
    rotationMode=BMM_Additive
    rotationSpace=BCS_BoneSpace
    alphaBoolBlend {blendInTime=0.1, blendOption=Linear, blendOutTime=0.2}
    boneToModify {boneName=dyn_kimono_back_01_r_MSTR}
    rotation {yaw=15}
   }
   n44 = RigidBody!(
    ComponentPose=Local To Component(LocalPose=Component To Local(ComponentPose=@n43))
    Alpha
    OverrideWorldGravity:Vector=Get Gravity Override()
    ExternalForce
   ) -> (Pose) {
    bOverrideWorldGravity=true
    overridePhysicsAsset=&/M_MED_TikiTorch_Physics
    simulationSpace=BaseBoneSpace
    alphaBoolBlend {blendOption=Linear}
    baseBoneRef {boneName=kimono_skirt_MSTR}
    componentAppliedLinearAccClamp=<1500 1500 900>
    componentLinearAccScale=<1 1 0.2>
    componentLinearVelScale=<1 1 0.3>
    overrideWorldGravity {z=-980}
   }
   n45 = Component To Local(ComponentPose=@n44)
   n46 = Get bRigidBodyEnabled()
   n47 = Get FrontEndIdle() -> (FrontEndIdle:bool)
   n48 = "Transform (Modify) Bone - Bone: dyn_kimono_back_01_l_MSTR"(
    ComponentPose=Local To Component(LocalPose=Blend Poses by bool(BlendPose_0=@n45, BlendPose_1=@n37, bActiveValue=@n46) {blendType=Linear})
    bAlphaBoolEnabled=@n47
    Translation="0,0,0"
    Rotation="0,5,0"
    Scale="1,1,1"
   ) {
    alphaInputType=Bool
    rotationMode=BMM_Additive
    rotationSpace=BCS_BoneSpace
    translationMode=BMM_Additive
    translationSpace=BCS_BoneSpace
    alphaBoolBlend {blendInTime=0.1, blendOption=Linear, blendOutTime=0.2}
    boneToModify {boneName=dyn_kimono_back_01_l_MSTR}
    rotation {yaw=5}
   }
   n49 = "Transform (Modify) Bone - Bone: dyn_kimono_back_01_MSTR"(ComponentPose=@n48, bAlphaBoolEnabled=@n47, Translation="0,0,0", Rotation="0,5,0", Scale="1,1,1") {
    alphaInputType=Bool
    rotationMode=BMM_Additive
    rotationSpace=BCS_BoneSpace
    translationMode=BMM_Additive
    translationSpace=BCS_BoneSpace
    alphaBoolBlend {blendInTime=0.1, blendOption=Linear, blendOutTime=0.2}
    boneToModify {boneName=dyn_kimono_back_01_MSTR}
    rotation {yaw=5}
   }
   n50 = "Transform (Modify) Bone - Bone: dyn_kimono_back_01_r_MSTR"(ComponentPose=@n49, bAlphaBoolEnabled=@n47, Translation="0,0,0", Rotation="0,5,0", Scale="1,1,1") {
    alphaInputType=Bool
    rotationMode=BMM_Additive
    rotationSpace=BCS_BoneSpace
    translationMode=BMM_Additive
    translationSpace=BCS_BoneSpace
    alphaBoolBlend {blendInTime=0.1, blendOption=Linear, blendOutTime=0.2}
    boneToModify {boneName=dyn_kimono_back_01_r_MSTR}
    rotation {yaw=5}
   }
   n51 = Component To Local(ComponentPose=@n50)
   Output Pose(Result=Component To Local(ComponentPose=Local To Component(LocalPose=@n51))) {layerGroup=None}
   n52 = Anim Dynamics - Bone: dyn_sleevs_l_01!(ComponentPose=Local To Component(LocalPose=Use cached pose 'PreSimSleeves'()), bAlphaBoolEnabled:bool=True, ExternalForce="0,0,0") -> (Pose) {
    alphaInputType=Bool
    angularDampingOverride=1
    bOverrideAngularDamping=true
    bOverrideLinearDamping=true
    bUseGravityOverride=true
    bUseSphericalLimits=true
    linearDampingOverride=1
    simulationSpace=RootRelative
    alphaBoolBlend {blendInTime=0.1, blendOption=Linear, blendOutTime=0.2}
    boundBone {boneName=dyn_sleevs_l_01}
    componentAppliedLinearAccClamp=<3000 3000 3000>
    componentLinearAccScale=<0.1 0.1 0.1>
    componentLinearVelScale=<0.2 0.2 0.2>
    gravityOverride {z=-980}
    physicsBodyDefinitions [{
      boundBone {boneName=dyn_sleevs_l_01}
      boxExtents=<15 15 15>
      constraintSetup {angularLimitsMax=<0 2 15>, angularLimitsMin=<0 -2 -15>, angularTarget=<0 0 0>, linearAxesMin=<-3 -3 -2>}
      localJointOffset=<1 0 0>
     }]
    relativeSpaceBone {boneName=lowerarm_l}
    sphericalLimits [
     {limitRadius=6.5, drivingBone {boneName=upperarm_twist_02_l}, sphereLocalOffset=<8 0 0>}
     {limitRadius=5, drivingBone {boneName=lowerarm_twist_02_l}, sphereLocalOffset=<0 0 0>}
    ]
   }
   n53 = Anim Dynamics - Bone: dyn_sleevs_l_02!(ComponentPose=@n52, bAlphaBoolEnabled:bool=True, ExternalForce="0,0,0") -> (Pose) {
    alphaInputType=Bool
    angularDampingOverride=1
    bOverrideAngularDamping=true
    bOverrideLinearDamping=true
    bUseGravityOverride=true
    bUseSphericalLimits=true
    linearDampingOverride=1
    simulationSpace=RootRelative
    alphaBoolBlend {blendInTime=0.1, blendOption=Linear, blendOutTime=0.2}
    boundBone {boneName=dyn_sleevs_l_02}
    componentAppliedLinearAccClamp=<3000 3000 3000>
    componentLinearAccScale=<0.1 0.1 0.1>
    componentLinearVelScale=<0.2 0.2 0.2>
    gravityOverride {z=-980}
    physicsBodyDefinitions [{
      boundBone {boneName=dyn_sleevs_l_02}
      boxExtents=<15 15 15>
      constraintSetup {angularLimitsMax=<0 2 15>, angularLimitsMin=<0 -2 -15>, angularTarget=<0 0 0>, linearAxesMin=<-7 -3 -3>}
      localJointOffset=<1 0 0>
     }]
    relativeSpaceBone {boneName=lowerarm_l}
    sphericalLimits [{limitRadius=6.5, drivingBone {boneName=upperarm_twist_02_l}, sphereLocalOffset=<8 0 0>}]
   }
   n54 = Anim Dynamics - Bone: dyn_sleevs_l_03!(ComponentPose=@n53, bAlphaBoolEnabled:bool=True, ExternalForce="0,0,0") -> (Pose) {
    alphaInputType=Bool
    angularDampingOverride=1
    bOverrideAngularDamping=true
    bOverrideLinearDamping=true
    bUseGravityOverride=true
    bUseSphericalLimits=true
    linearDampingOverride=1
    simulationSpace=RootRelative
    alphaBoolBlend {blendInTime=0.1, blendOption=Linear, blendOutTime=0.2}
    boundBone {boneName=dyn_sleevs_l_03}
    componentAppliedLinearAccClamp=<3000 3000 3000>
    componentLinearAccScale=<0.1 0.1 0.1>
    componentLinearVelScale=<0.2 0.2 0.2>
    gravityOverride {z=-980}
    physicsBodyDefinitions [{
      boundBone {boneName=dyn_sleevs_l_03}
      boxExtents=<15 15 15>
      constraintSetup {angularLimitsMax=<0 2 15>, angularLimitsMin=<0 -2 -15>, angularTarget=<0 0 0>, linearAxesMin=<-7 -3 -3>}
      localJointOffset=<1 0 0>
     }]
    relativeSpaceBone {boneName=lowerarm_l}
    sphericalLimits [
     {limitRadius=6.5, drivingBone {boneName=upperarm_twist_02_l}, sphereLocalOffset=<8 0 0>}
     {limitRadius=5, drivingBone {boneName=lowerarm_twist_02_l}, sphereLocalOffset=<0 0 0>}
    ]
   }
   n55 = Anim Dynamics - Bone: dyn_sleevs_l_04!(ComponentPose=@n54, bAlphaBoolEnabled:bool=True, ExternalForce="0,0,0") -> (Pose) {
    alphaInputType=Bool
    angularDampingOverride=1
    bOverrideAngularDamping=true
    bOverrideLinearDamping=true
    bUseGravityOverride=true
    bUseSphericalLimits=true
    linearDampingOverride=1
    simulationSpace=RootRelative
    alphaBoolBlend {blendInTime=0.1, blendOption=ExpOut, blendOutTime=0.2}
    boundBone {boneName=dyn_sleevs_l_04}
    componentAppliedLinearAccClamp=<3000 3000 3000>
    componentLinearAccScale=<0.1 0.1 0.1>
    componentLinearVelScale=<0.2 0.2 0.2>
    gravityOverride {z=-980}
    physicsBodyDefinitions [{
      boundBone {boneName=dyn_sleevs_l_04}
      boxExtents=<15 15 15>
      constraintSetup {angularLimitsMax=<0 2 15>, angularLimitsMin=<0 -2 -15>, angularTarget=<0 0 0>, linearAxesMin=<-3 -3 0>}
      localJointOffset=<0.2 0 0>
     }]
    relativeSpaceBone {boneName=lowerarm_l}
    sphericalLimits [
     {limitRadius=6.5, drivingBone {boneName=upperarm_twist_02_l}, sphereLocalOffset=<8 0 0>}
     {limitRadius=3.5, drivingBone {boneName=lowerarm_twist_02_l}, sphereLocalOffset=<1 0 0>}
    ]
   }
   n56 = "Transform (Modify) Bone - Bone: dyn_sleevs_l_01_MSTR"(ComponentPose=@n55, Alpha=@n8, Translation="-4,0,0", Rotation="0,0,0", Scale="1,1,1") {
    rotationMode=BMM_Additive
    rotationSpace=BCS_BoneSpace
    translationMode=BMM_Additive
    translationSpace=BCS_BoneSpace
    alphaBoolBlend {blendInTime=0.1, blendOption=Linear, blendOutTime=0.1}
    boneToModify {boneName=dyn_sleevs_l_01_MSTR}
    translation {x=-4}
   }
   n57 = "Transform (Modify) Bone - Bone: dyn_sleevs_l_04_MSTR"(ComponentPose=@n56, Alpha=@n8, Translation="-20,0,0", Rotation="0,15,0", Scale="1,1,1") {
    rotationMode=BMM_Additive
    rotationSpace=BCS_BoneSpace
    translationMode=BMM_Additive
    translationSpace=BCS_BoneSpace
    alphaBoolBlend {blendInTime=0.1, blendOption=Linear, blendOutTime=0.1}
    boneToModify {boneName=dyn_sleevs_l_04_MSTR}
    rotation {yaw=15}
    translation {x=-20}
   }
   n58 = Anim Dynamics - Bone: dyn_sleevs_r_01!(ComponentPose=@n57, bAlphaBoolEnabled:bool=True, ExternalForce="0,0,0") -> (Pose) {
    alphaInputType=Bool
    angularDampingOverride=1
    bOverrideAngularDamping=true
    bOverrideLinearDamping=true
    bUseGravityOverride=true
    bUseSphericalLimits=true
    linearDampingOverride=1
    simulationSpace=RootRelative
    alphaBoolBlend {blendInTime=0.1, blendOption=Linear, blendOutTime=0.2}
    boundBone {boneName=dyn_sleevs_r_01}
    componentAppliedLinearAccClamp=<3000 3000 3000>
    componentLinearAccScale=<0.1 0.1 0.1>
    componentLinearVelScale=<0.2 0.2 0.2>
    gravityOverride {z=-980}
    physicsBodyDefinitions [{
      boundBone {boneName=dyn_sleevs_r_01}
      boxExtents=<15 15 15>
      constraintSetup {angularLimitsMax=<0 2 15>, angularLimitsMin=<0 -2 -15>, angularTarget=<0 0 0>, linearAxesMax=<0 0 2>, linearAxesMin=<-3 -5 -2>}
      localJointOffset=<1 0 0>
     }]
    relativeSpaceBone {boneName=lowerarm_r}
    sphericalLimits [
     {limitRadius=6.5, drivingBone {boneName=upperarm_twist_02_r}, sphereLocalOffset=<-8 0 0>}
     {limitRadius=5, drivingBone {boneName=lowerarm_twist_02_r}, sphereLocalOffset=<0 0 0>}
    ]
   }
   n59 = Anim Dynamics - Bone: dyn_sleevs_r_02!(ComponentPose=@n58, bAlphaBoolEnabled:bool=True, ExternalForce="0,0,0") -> (Pose) {
    alphaInputType=Bool
    angularDampingOverride=1
    bOverrideAngularDamping=true
    bOverrideLinearDamping=true
    bUseGravityOverride=true
    bUseSphericalLimits=true
    linearDampingOverride=1
    simulationSpace=RootRelative
    alphaBoolBlend {blendInTime=0.1, blendOption=Linear, blendOutTime=0.2}
    boundBone {boneName=dyn_sleevs_r_02}
    componentAppliedLinearAccClamp=<3000 3000 3000>
    componentLinearAccScale=<0.1 0.1 0.1>
    componentLinearVelScale=<0.2 0.2 0.2>
    gravityOverride {z=-980}
    physicsBodyDefinitions [{
      boundBone {boneName=dyn_sleevs_r_02}
      boxExtents=<15 15 15>
      constraintSetup {angularLimitsMax=<0 2 15>, angularLimitsMin=<0 -2 -15>, angularTarget=<0 0 0>, linearAxesMax=<0 0 3>, linearAxesMin=<-7 -3 0>}
      localJointOffset=<1 0 0>
     }]
    relativeSpaceBone {boneName=lowerarm_r}
    sphericalLimits [{limitRadius=6.5, drivingBone {boneName=upperarm_twist_02_r}, sphereLocalOffset=<-8 0 0>}]
   }
   n60 = Anim Dynamics - Bone: dyn_sleevs_r_03!(ComponentPose=@n59, bAlphaBoolEnabled:bool=True, ExternalForce="0,0,0") -> (Pose) {
    alphaInputType=Bool
    angularDampingOverride=1
    bOverrideAngularDamping=true
    bOverrideLinearDamping=true
    bUseGravityOverride=true
    bUseSphericalLimits=true
    linearDampingOverride=1
    simulationSpace=RootRelative
    alphaBoolBlend {blendInTime=0.1, blendOption=Linear, blendOutTime=0.2}
    boundBone {boneName=dyn_sleevs_r_03}
    componentAppliedLinearAccClamp=<3000 3000 3000>
    componentLinearAccScale=<0.1 0.1 0.1>
    componentLinearVelScale=<0.2 0.2 0.2>
    gravityOverride {z=-980}
    physicsBodyDefinitions [{
      boundBone {boneName=dyn_sleevs_r_03}
      boxExtents=<15 15 15>
      constraintSetup {angularLimitsMax=<0 2 15>, angularLimitsMin=<0 -2 -15>, angularTarget=<0 0 0>, linearAxesMax=<0 0 3>, linearAxesMin=<-7 -3 0>}
      localJointOffset=<1 0 0>
     }]
    relativeSpaceBone {boneName=lowerarm_r}
    sphericalLimits [
     {limitRadius=6.5, drivingBone {boneName=upperarm_twist_02_r}, sphereLocalOffset=<-8 0 0>}
     {limitRadius=5, drivingBone {boneName=lowerarm_twist_02_r}, sphereLocalOffset=<0 0 0>}
    ]
   }
   n61 = Anim Dynamics - Bone: dyn_sleevs_r_04!(ComponentPose=@n60, bAlphaBoolEnabled:bool=True, ExternalForce="0,0,0") -> (Pose) {
    alphaInputType=Bool
    angularDampingOverride=1
    bOverrideAngularDamping=true
    bOverrideLinearDamping=true
    bUseGravityOverride=true
    bUseSphericalLimits=true
    linearDampingOverride=1
    simulationSpace=RootRelative
    alphaBoolBlend {blendInTime=0.1, blendOption=ExpOut, blendOutTime=0.2}
    boundBone {boneName=dyn_sleevs_r_04}
    componentAppliedLinearAccClamp=<3000 3000 3000>
    componentLinearAccScale=<0.1 0.1 0.1>
    componentLinearVelScale=<0.2 0.2 0.2>
    gravityOverride {z=-980}
    physicsBodyDefinitions [{
      boundBone {boneName=dyn_sleevs_r_04}
      boxExtents=<15 15 15>
      constraintSetup {angularLimitsMax=<0 2 15>, angularLimitsMin=<0 -2 -15>, angularTarget=<0 0 0>, linearAxesMin=<-3 -3 0>}
      localJointOffset=<0.2 0 0>
     }]
    relativeSpaceBone {boneName=lowerarm_r}
    sphericalLimits [
     {limitRadius=6.5, drivingBone {boneName=upperarm_twist_02_r}, sphereLocalOffset=<-8 0 0>}
     {limitRadius=3.5, drivingBone {boneName=lowerarm_twist_02_r}, sphereLocalOffset=<-1 0 0>}
    ]
   }
   n62 = "Transform (Modify) Bone - Bone: dyn_sleevs_r_01_MSTR"(ComponentPose=@n61, Alpha=@n7, Translation="-4,0,0", Rotation="0,0,0", Scale="1,1,1") {
    rotationMode=BMM_Additive
    rotationSpace=BCS_BoneSpace
    translationMode=BMM_Additive
    translationSpace=BCS_BoneSpace
    alphaBoolBlend {blendInTime=0.1, blendOption=Linear, blendOutTime=0.1}
    boneToModify {boneName=dyn_sleevs_r_01_MSTR}
    translation {x=-4}
   }
   n63 = "Transform (Modify) Bone - Bone: dyn_sleevs_r_04_MSTR"(ComponentPose=@n62, Alpha=@n7, Translation="-20,0,0", Rotation="0,15,0", Scale="1,1,1") {
    rotationMode=BMM_Additive
    rotationSpace=BCS_BoneSpace
    translationMode=BMM_Additive
    translationSpace=BCS_BoneSpace
    alphaBoolBlend {blendInTime=0.1, blendOption=Linear, blendOutTime=0.1}
    boneToModify {boneName=dyn_sleevs_r_04_MSTR}
    rotation {yaw=15}
    translation {x=-20}
   }
   n64 = Get bIsSkyDiving()
   n65 = "Transform (Modify) Bone - Bone: dyn_sleevs_r_01_MSTR"!(ComponentPose=@n63, bAlphaBoolEnabled:bool=@n64, Translation="-2,0,0", Rotation="0,0,0", Scale="1,1,1") -> (Pose) {
    alphaInputType=Bool
    rotationMode=BMM_Additive
    rotationSpace=BCS_BoneSpace
    translationMode=BMM_Additive
    translationSpace=BCS_BoneSpace
    alphaBoolBlend {blendInTime=0.1, blendOption=Linear, blendOutTime=0.2}
    boneToModify {boneName=dyn_sleevs_r_01_MSTR}
    translation {x=-2}
   }
   n66 = "Transform (Modify) Bone - Bone: dyn_sleevs_r_04_MSTR"!(ComponentPose=@n65, bAlphaBoolEnabled:bool=@n64, Translation="-2,0,0", Rotation="0,0,0", Scale="1,1,1") -> (Pose) {
    alphaInputType=Bool
    rotationMode=BMM_Additive
    rotationSpace=BCS_BoneSpace
    translationMode=BMM_Additive
    translationSpace=BCS_BoneSpace
    alphaBoolBlend {blendInTime=0.1, blendOption=Linear, blendOutTime=0.2}
    boneToModify {boneName=dyn_sleevs_r_04_MSTR}
    translation {x=-2}
   }
   n67 = "Transform (Modify) Bone - Bone: dyn_sleevs_l_01_MSTR"!(ComponentPose=@n66, bAlphaBoolEnabled:bool=@n64, Translation="-2,0,0", Rotation="0,0,0", Scale="1,1,1") -> (Pose) {
    alphaInputType=Bool
    rotationMode=BMM_Additive
    rotationSpace=BCS_BoneSpace
    translationMode=BMM_Additive
    translationSpace=BCS_BoneSpace
    alphaBoolBlend {blendInTime=0.1, blendOption=Linear, blendOutTime=0.2}
    boneToModify {boneName=dyn_sleevs_l_01_MSTR}
    translation {x=-2}
   }
   n68 = "Transform (Modify) Bone - Bone: dyn_sleevs_l_04_MSTR"!(ComponentPose=@n67, bAlphaBoolEnabled:bool=@n64, Translation="-2,0,0", Rotation="0,0,0", Scale="1,1,1") -> (Pose) {
    alphaInputType=Bool
    rotationMode=BMM_Additive
    rotationSpace=BCS_BoneSpace
    translationMode=BMM_Additive
    translationSpace=BCS_BoneSpace
    alphaBoolBlend {blendInTime=0.1, blendOption=Linear, blendOutTime=0.2}
    boneToModify {boneName=dyn_sleevs_l_04_MSTR}
    translation {x=-2}
   }
   Save cached pose 'First_stage'(Pose=Blend Poses by bool(
     BlendPose_0=Component To Local(ComponentPose=@n68)
     BlendPose_1=Use cached pose 'PreSimSleeves'()
     bActiveValue=Get Not in Frontend idle() -> (Not in Frontend idle:bool)
    ) {blendType=Linear}) {cacheName=First_stage}
   n69 = Anim Dynamics - Bone: dyn_kimono_front_02_l1!(
    ComponentPose=Local To Component(LocalPose=Copy Pose From Mesh() {bUseAttachedParent=true})
    bAlphaBoolEnabled:bool=True
    ExternalForce="0,0,0"
   ) -> (Pose) {
    alphaInputType=Bool
    angularDampingOverride=1
    bOverrideAngularDamping=true
    bUseGravityOverride=true
    simulationSpace=RootRelative
    alphaBoolBlend {blendInTime=0.1, blendOption=Linear, blendOutTime=0.2}
    boundBone {boneName=dyn_kimono_front_02_l1}
    componentAppliedLinearAccClamp=<3000 3000 3000>
    componentLinearAccScale=<0.1 0.1 0.1>
    componentLinearVelScale=<0.2 0.2 0.2>
    gravityOverride {z=-980}
    physicsBodyDefinitions [{
      boundBone {boneName=dyn_kimono_front_02_l1}
      boxExtents=<8 2 8>
      constraintSetup {angularLimitsMax=<0 2 2>, angularLimitsMin=<0 -2 -30>, angularTarget=<0 0 0>}
      localJointOffset=<1 0 0>
     }]
   }
   n70 = Anim Dynamics - Bone: dyn_kimono_front_02_r1!(ComponentPose=@n69, bAlphaBoolEnabled:bool=True, ExternalForce="0,0,0") -> (Pose) {
    alphaInputType=Bool
    angularDampingOverride=1
    bOverrideAngularDamping=true
    bUseGravityOverride=true
    simulationSpace=RootRelative
    alphaBoolBlend {blendInTime=0.1, blendOption=Linear, blendOutTime=0.2}
    boundBone {boneName=dyn_kimono_front_02_r1}
    componentAppliedLinearAccClamp=<3000 3000 3000>
    componentLinearAccScale=<0.1 0.1 0.1>
    componentLinearVelScale=<0.2 0.2 0.2>
    gravityOverride {z=-980}
    physicsBodyDefinitions [{
      boundBone {boneName=dyn_kimono_front_02_r1}
      boxExtents=<8 2 8>
      constraintSetup {angularLimitsMax=<0 2 2>, angularLimitsMin=<0 -2 -30>, angularTarget=<0 0 0>}
      localJointOffset=<1 0 0>
     }]
   }
   Save cached pose 'PreSimSleeves'(Pose=Component To Local(ComponentPose=@n70)) {cacheName=PreSimSleeves}
   Reroute Node(InputPin:Wildcard) -> (OutputPin:Wildcard)
  ]}
 EventGraph {nodes [
   Event BlueprintUpdateAnimation()
   > Set Aim_offset_r(Aim_offset_r:float=GetAlphaFromReferenceTransformDelta(SkeletalMeshComponent=Get Skel Mesh(), BoneName=upperarm_r, BaseName=lowerarm_r, InputRange=(X=-40,Y=-140), TwistAxis="1,0,0")) -> (Output_Get:float)
   > Set Aim_offset_l(Aim_offset_l:float=GetAlphaFromReferenceTransformDelta(SkeletalMeshComponent=Get Skel Mesh(), BoneName=upperarm_l, BaseName=lowerarm_l, InputRange=(X=-40,Y=-140), TwistAxis="1,0,0")) -> (Output_Get:float)
   > n1 = Gravity Override State Function DEPRECATED(CharacterPartAnimInstance=Self-Reference(), GameTimeInSeconds=GetGameTimeInSeconds(), GravityOverrideParams=Get Gravity Override Params())
   > Set Gravity Override(Gravity Override=@n1)
   n2 = Get bIsCrouching()
   n3 = Get bIsSliding()
   n4 = Get bIsSkydiveDiveMode()
   > Set AltSkinMovement(AltSkinMovement:bool=NOT Boolean(A=OR Boolean!(A=@n2, B=@n4, C:bool=@n3) -> (ReturnValue))) -> (Output_Get:bool)
   n5 = Get bIsPlayingEmote()
   > Set FrontEndIdle(FrontEndIdle:bool=AND Boolean(A=Get bIsInFrontEnd(), B=NOT Boolean(A=@n5))) -> (Output_Get:bool)
   > Set Not in Frontend idle(Not in Frontend idle:bool=NOT Boolean(A=Get FrontEndIdle())) -> (Output_Get:bool)
   n6 = Event OnCharacterCustomizationLoaded()
   > Set Skel Mesh(Skel Mesh=Get Array Item with EFort Custom Part Type(Array=@n6.CustomMeshParts))
  ]}
]
variables [
 Aim_offset_l {type=number}
 Aim_offset_r {type=number}
 AltSkinMovement {type=boolean}
 FrontEndIdle {type=boolean}
 Gravity Override {type=Vector}
 Gravity Override Params {
  type=GravityOverrideParamsStruct
  default {
   skydive_DivingDown {gravityOverride=<-3000 -500 -500>}
   skydive_FloatingDown {gravityOverride=<0 -700 -1500>}
   skydive_GoingUp {gravityOverride=<0 300 -1500>}
   skydive_Parachuting {gravityOverride=<-1500 -2000 -980>}
  }
 }
 Not in Frontend idle {type=boolean}
 Skel Mesh {type=SkeletalMeshComponent}
]
